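# Emissive screen material for the M_MED robot backpack, showing a kill counter or a reticle on a flickery CRT-style display.
# Kills-Tens and Kills-Ones offset UVs into a number atlas, with parameters blending between two-digit, single-digit and reticle layouts.
# A radial gradient warps the UVs, scrolling noise jitters the glyphs, and tiled scanline masks texture them.
# Two tints color background and glyphs inside a box mask; the optional sine flash reaches only the background.
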Material /BRCosmetics/Accessories/FORT_Backpacks/M_MED_Robot_Backpack/Materials/M_MED_Robot_Backpack_Screen
graph {
 properties {
  bUseMaterialAttributes=true
  bUsedWithMeshDeformer=true
  bUsedWithSkeletalMesh=true
  pixelDepthOffsetMode=PDOM_Legacy
  shadingModel=MSM_FromMaterialExpression
  usedShadingModels=Default Lit
 }
 nodes [
  Constant() {r=0.7}
  Constant() {r=2}
  Constant() {r=-0.8}
  Constant() {r=2}
  Constant() {r=1.1}
  Constant() {r=1.34}
  n1 = Time()
  n2 = ScalarParameter() {defaultValue=2.5, parameterName=Flash}
  n3 = ConstantBiasScale(None=Sine(None=Multiply(A=@n1, B=@n2)))
  n4 = ScalarParameter() {parameterName=Kills-Tens}
  n5 = Constant() {r=0.0938}
  n6 = TextureCoordinate()
  n7 = ScalarParameter() {defaultValue=1.2, parameterName=Horiz-01_30}
  n8 = ScalarParameter() {defaultValue=1.2, parameterName=Horiz-01_31}
  n9 = Add(
   A=MaterialFunctionCall(
    UVs=@n6
    Texture Scale=MaterialFunctionCall(X=@n7, Y=@n8) {materialFunction=&MakeFloat2}.Result
   ) {materialFunction=&ScaleUVsByCenter}.UVs
   B=ScalarParameter() {defaultValue=-0.04, parameterName=Horiz-01_15}
  )
  n10 = MaterialFunctionCall(
   UVs=TextureCoordinate()
   CenterPosition=Constant2Vector() {g=0.5, r=0.5}
   Radius=ScalarParameter() {defaultValue=1.2, parameterName=Horiz-01_2}
   Density=ScalarParameter() {defaultValue=3, parameterName=Horiz-01_17}
  ) {materialFunction=&RadialGradientExponential}
  n11 = TextureCoordinate()
  n12 = ScalarParameter() {defaultValue=0.25, parameterName=Horiz-01_13}
  n13 = Add(
   A=Multiply(
    A=MaterialFunctionCall(UVs=@n11, Texture Scale=@n12) {materialFunction=&ScaleUVsByCenter}.UVs
    B=ScalarParameter() {defaultValue=0.5, parameterName=Horiz-01_12}
   )
   B=@n9
  )
  n14 = LinearInterpolate(A=@n13, B=@n9, Alpha=@n10.RadialGradientExponential)
  n15 = MaterialFunctionCall(Float2=@n14) {materialFunction=&BreakOutFloat2Components}
  n16 = ScalarParameter() {defaultValue=0.22, parameterName=Horiz-01}
  n17 = Multiply(A=@n15.R, B=@n16)
  n18 = Add(A=ScalarParameter() {defaultValue=0.015, parameterName=Horiz-01_1}, B=@n17)
  n19 = ScalarParameter() {defaultValue=0.1, parameterName=Horiz-01_4}
  n20 = Add(A=Multiply(A=Time(), B=Constant() {r=-1}), B=@n19)
  n21 = ScalarParameter() {defaultValue=0.2, parameterName=Horiz-01_28}
  n22 = MaterialFunctionCall(Float2=TextureCoordinate()) {materialFunction=&BreakOutFloat2Components}
  n23 = MaterialFunctionCall(X=Add(A=Multiply(A=@n20, B=@n19), B=Multiply(A=@n21, B=@n22.R)), Y=Add(A=@n20, B=@n22.G)) {materialFunction=&MakeFloat2}
  n24 = Saturate(None=LinearInterpolate(
    A=ScalarParameter() {defaultValue=-0.2, parameterName=Horiz-01_24}
    B=ScalarParameter() {defaultValue=1.1, parameterName=Horiz-01_25}
    Alpha=TextureSample(UVs=@n23.Result) {samplerType=SAMPLERTYPE_Grayscale, texture=&LowResBlurredNoise}.R
   ))
  n25 = Multiply(A=@n24, B=ScalarParameter() {defaultValue=0.012, parameterName=Horiz-01_26})
  n26 = Add(A=@n25, B=Add(A=Multiply(A=@n4, B=@n5), B=@n18))
  n27 = Add(A=ScalarParameter() {defaultValue=0.39, parameterName=VertOffset}, B=Multiply(A=@n15.G, B=@n16))
  n28 = Add(A=@n25, B=@n18)
  n29 = ScalarParameter() {defaultValue=0.2, parameterName=Horiz-01_8}
  n30 = Add(A=ScalarParameter() {defaultValue=-0.015, parameterName=VertOffset_1}, B=Multiply(A=@n15.R, B=@n29))
  n31 = Add(A=@n25, B=@n30)
  n32 = Add(A=ScalarParameter() {defaultValue=0.4, parameterName=VertOffset_2}, B=Multiply(A=@n15.G, B=@n29))
  n33 = ScalarParameter() {defaultValue=1, parameterName=UseSingleDigit}
  n34 = TextureSample(UVs=LinearInterpolate(
    A=MaterialFunctionCall(X=@n28, Y=@n27) {materialFunction=&MakeFloat2}.Result
    B=MaterialFunctionCall(X=@n31, Y=@n32) {materialFunction=&MakeFloat2}.Result
    Alpha=@n33
   )) {samplerType=SAMPLERTYPE_LinearColor, texture=/Game/Accessories/FORT_Backpacks/M_MED_Robot_Backpack/Textures/T_SimpleNumbers_Square}
  n35 = Multiply(
   A=TextureSample(UVs=MaterialFunctionCall(X=@n26, Y=@n27) {materialFunction=&MakeFloat2}.Result) {samplerType=SAMPLERTYPE_LinearColor, texture=&T_SimpleNumbers_Square}.R
   B=@n34.B
  )
  n36 = Multiply(A=ScalarParameter() {parameterName=Kills-Ones}, B=@n5)
  n37 = Add(A=ScalarParameter() {defaultValue=-0.063, parameterName=Horiz-01_3}, B=@n17)
  n38 = MaterialFunctionCall(X=Add(A=@n25, B=Add(A=@n36, B=@n37)), Y=@n27) {materialFunction=&MakeFloat2}
  n39 = MaterialFunctionCall(X=Add(A=@n25, B=Add(A=@n36, B=@n30)), Y=@n32) {materialFunction=&MakeFloat2}
  n40 = ScalarParameter() {defaultValue=0.2, parameterName=Horiz-01_16}
  n41 = Add(A=@n25, B=Add(A=ScalarParameter() {defaultValue=0.01, parameterName=VertOffset_4}, B=Multiply(A=@n15.R, B=@n40)))
  n42 = Add(A=ScalarParameter() {defaultValue=0.68, parameterName=VertOffset_3}, B=Multiply(A=@n15.G, B=@n40))
  n43 = ScalarParameter() {defaultValue=1, parameterName=UseReticle}
  n44 = TextureSample(UVs=LinearInterpolate(
    A=LinearInterpolate(A=@n38.Result, B=@n39.Result, Alpha=@n33)
    B=MaterialFunctionCall(X=@n41, Y=@n42) {materialFunction=&MakeFloat2}.Result
    Alpha=@n43
   )) {samplerType=SAMPLERTYPE_LinearColor, texture=&T_SimpleNumbers_Square}
  n45 = Multiply(A=@n44.R, B=OneMinus(None=@n34.B))
  n46 = LinearInterpolate(A=LinearInterpolate(A=Add(A=@n35, B=@n45), B=Multiply(A=@n44.R, B=@n34.B), Alpha=@n33), B=@n44.R, Alpha=@n43)
  n47 = ScalarParameter() {defaultValue=3, parameterName=mmh}
  n48 = TextureSample(UVs=Multiply(A=ScalarParameter() {defaultValue=14, parameterName=ScanlineTile_2}, B=@n14)) {samplerType=SAMPLERTYPE_LinearColor, texture=&T_SimpleNumbers_Square}
  n49 = Saturate(None=Add(A=OneMinus(None=@n48.G), B=ScalarParameter() {defaultValue=-0.5, parameterName=Horiz-01_18}))
  n50 = ScalarParameter() {defaultValue=0.8, parameterName=Horiz-01_33}
  n51 = TextureSample(UVs=Multiply(A=@n14, B=ScalarParameter() {defaultValue=14, parameterName=ScanlineTile})) {samplerType=SAMPLERTYPE_LinearColor, texture=/Game/Global/Textures/Default/Masks_and_Gradients/T_VertGradient}
  n52 = ScalarParameter() {defaultValue=0.4, parameterName=Horiz-01_23}
  n53 = Multiply(A=Saturate(None=Add(A=@n49, B=@n50)), B=Saturate(None=Add(A=@n51.R, B=@n52)))
  n54 = Add(A=Multiply(A=@n24, B=ScalarParameter() {defaultValue=0.04, parameterName=Horiz-01_14}), B=Multiply(A=@n53, B=Power(Base=@n46, Exp=@n47)))
  n55 = ScalarParameter() {parameterName=FlashControl}
  LinearInterpolate(A=@n54, B=Multiply(A=@n54, B=@n3), Alpha=@n55)
  n56 = Add(A=Power(Base=@n10.RadialGradientExponential, Exp=ScalarParameter() {defaultValue=100, parameterName=Horiz-01_5}), B=@n53)
  n57 = LinearInterpolate(A=@n56, B=Multiply(A=@n56, B=@n3), Alpha=@n55)
  n58 = VectorParameter() {parameterName=Color-A, defaultValue=<rgba 0 0.0814 0.15 1>}
  n59 = VectorParameter() {parameterName=Color-A_1, defaultValue=<rgba 0.25 1.493 2.5 1>}
  n60 = ScalarParameter() {defaultValue=0.8, parameterName=Horiz-01_9}
  n61 = ScalarParameter() {defaultValue=0.8, parameterName=Horiz-01_10}
  n62 = Multiply(
   A=MaterialFunctionCall!(
    A=LinearInterpolate(A=@n13, B=@n6, Alpha=@n10.RadialGradientExponential)
    B=Constant2Vector() {g=0.5, r=0.5}
    Bounds=MaterialFunctionCall(X=@n60, Y=@n61) {materialFunction=&MakeFloat2}.Result
    Edge Falloff=ScalarParameter() {defaultValue=0.15, parameterName=Horiz-01_6}
   ) {materialFunction=/Engine/Functions/Engine_MaterialFunctions01/Gradient/BoxMask-2D}.Result
   B=Add(A=Multiply(A=@n57, B=@n58.RGB), B=Multiply(A=@n59.RGB, B=@n54))
  )
  MaterialFunctionCall(In=SetMaterialAttributes!(
    MaterialAttributes
    Emissive Color=NamedRerouteDeclaration(None=@n62) {name=emissive, nodeColor=<rgba 0.7109 1 0 1>}
    Shading Model=ShadingModel()
   ) {attributeSetTypes [#1, #2]}) {materialFunction=&MF_TechArtOverrides_accessories}
  TextureCoordinate()
 ]
}
