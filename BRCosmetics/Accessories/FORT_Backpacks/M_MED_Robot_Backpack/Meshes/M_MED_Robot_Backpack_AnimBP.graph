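# Animation Blueprint for the M_MED_Robot backpack cosmetic, adding secondary physics motion to its hoses and pull joint.
# The input pose gets a small additive offset and roll on offsetAll, then passes through four chained Anim Dynamics nodes in component space.
# The low hose and pull joints use limited angular springs; the two top hose joints use stiff, tightly limited linear springs with far weaker gravity.
# The event graph holds no logic.
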AnimBlueprint /BRCosmetics/Accessories/FORT_Backpacks/M_MED_Robot_Backpack/Meshes/M_MED_Robot_Backpack_AnimBP
target_skeleton=&/M_MED_Robot_Backpack_Skeleton
graphs [
 AnimGraph {nodes [
   AnimGraphNode_SubInput_0 = Input Pose()
   n1 = "Transform (Modify) Bone - Bone: offsetAll"!(ComponentPose=Local To Component(LocalPose=@AnimGraphNode_SubInput_0), Alpha=1, Translation="0,-0.5,1", Rotation="0,0,4") -> (Pose) {
    rotationMode=BMM_Additive
    translationMode=BMM_Additive
    alphaBoolBlend {blendOption=Linear}
    boneToModify {boneName=offsetAll}
    rotation {roll=4}
    translation {y=-0.5, z=1}
   }
   n2 = Anim Dynamics - Bone: dyn_lowhose_jnt(ComponentPose=@n1, Alpha=1, ExternalForce="0,0,0") {
    angularDampingOverride=0.995
    angularSpringConstant=300
    bAngularSpring=true
    bGravityOverrideInSimSpace=true
    bOverrideLinearDamping=true
    bUseGravityOverride=true
    linearDampingOverride=0.995
    linearSpringConstant=3
    alphaBoolBlend {blendOption=Linear}
    boundBone {boneName=dyn_lowhose_jnt}
    componentAppliedLinearAccClamp=<300 300 300>
    componentLinearAccScale=<1 1 1>
    componentLinearVelScale=<1 1 1>
    gravityOverride {z=-981}
    physicsBodyDefinitions [{
      sphereCollisionRadius=0
      boundBone {boneName=dyn_lowhose_jnt}
      boxExtents=<12 12 4>
      constraintSetup {angularLimitsMax=<0 15 0>, angularLimitsMin=<0 -15 0>, angularTarget=<0 0 0>}
      localJointOffset=<0 0 -6>
     }]
   }
   n3 = Anim Dynamics - Bone: dyn_tophoseA_jnt(ComponentPose=@n2, Alpha=1, ExternalForce="0,0,0") {
    bGravityOverrideInSimSpace=true
    bLinearSpring=true
    bOverrideLinearDamping=true
    bUseGravityOverride=true
    linearDampingOverride=0.999
    linearSpringConstant=1100
    alphaBoolBlend {blendOption=Linear}
    boundBone {boneName=dyn_tophoseA_jnt}
    componentAppliedLinearAccClamp=<300 300 300>
    componentLinearAccScale=<1 1 1>
    componentLinearVelScale=<1 1 1>
    gravityOverride {z=-9.81}
    physicsBodyDefinitions [{
      sphereCollisionRadius=0
      boundBone {boneName=dyn_tophoseA_jnt}
      boxExtents=<12 12 12>
      constraintSetup {angularLimitsMax=<0 0 0>, angularLimitsMin=<0 0 0>, angularTarget=<0 0 0>, linearAxesMax=<0 0.2 0.2>, linearAxesMin=<0 -0.2 -0.2>}
      localJointOffset=<0 0 0>
     }]
   }
   n4 = Anim Dynamics - Bone: dyn_tophoseB_jnt(ComponentPose=@n3, Alpha=1, ExternalForce="0,0,0") {
    bGravityOverrideInSimSpace=true
    bLinearSpring=true
    bOverrideLinearDamping=true
    bUseGravityOverride=true
    linearDampingOverride=0.995
    linearSpringConstant=1150
    alphaBoolBlend {blendOption=Linear}
    boundBone {boneName=dyn_tophoseB_jnt}
    componentAppliedLinearAccClamp=<300 300 300>
    componentLinearAccScale=<1 1 1>
    componentLinearVelScale=<1 1 1>
    gravityOverride {z=-9.81}
    physicsBodyDefinitions [{
      sphereCollisionRadius=0
      boundBone {boneName=dyn_tophoseB_jnt}
      boxExtents=<12 13 11>
      constraintSetup {angularLimitsMax=<0 0 0>, angularLimitsMin=<0 0 0>, angularTarget=<0 0 0>, linearAxesMax=<0 0.2 0.2>, linearAxesMin=<0 -0.2 -0.2>}
      localJointOffset=<0 0 0>
     }]
   }
   n5 = Anim Dynamics - Bone: dyn_pull_jnt(ComponentPose=@n4, Alpha=1, ExternalForce="0,0,0") {
    angularDampingOverride=0.995
    angularSpringConstant=500
    bAngularSpring=true
    bGravityOverrideInSimSpace=true
    bOverrideAngularDamping=true
    bUseGravityOverride=true
    linearDampingOverride=0.995
    linearSpringConstant=600
    alphaBoolBlend {blendOption=Linear}
    boundBone {boneName=dyn_pull_jnt}
    componentAppliedLinearAccClamp=<300 300 300>
    componentLinearAccScale=<1 1 1>
    componentLinearVelScale=<1 1 1>
    gravityOverride {z=-981}
    physicsBodyDefinitions [{
      sphereCollisionRadius=0
      boundBone {boneName=dyn_pull_jnt}
      boxExtents=<3 12 12>
      constraintSetup {angularLimitsMax=<0 2 3>, angularLimitsMin=<0 -2 -3>, angularTarget=<0 0 0>}
      localJointOffset=<3 0 3>
     }]
   }
   Output Pose(Result=Component To Local(ComponentPose=@n5)) {layerGroup=None}
  ]}
 EventGraph {nodes [Event BlueprintUpdateAnimation()]}
]
variables []
